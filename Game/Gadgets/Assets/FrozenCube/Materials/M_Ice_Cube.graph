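# Two-sided ice material for the Frozen Cube gadget, with Blueprint-driven damage cracks and a hit wobble.
# Three parallax crack samples, retiled at damage thresholds, blend deep and surface colors, with spherical harmonics decoded from vertex color tinting depth and adding opacity.
# Triplanar snow, masked by local bounds and cracks, drives color, roughness, normals and opacity.
# Fresnel emissive, switchable refraction, height-masked wobble offset and optional per-face opacity complete it.
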Material /Game/Gadgets/Assets/FrozenCube/Materials/M_Ice_Cube
graph {
 properties {
  bCastRayTracedShadows=false
  bUseMaterialAttributes=true
  ditherOpacityMask=true
  pixelDepthOffsetMode=PDOM_Legacy
  previewMesh=/Game/Gadgets/Assets/FrozenCube/Meshes/Frozen_Cube
  refractionMethod=RM_IndexOfRefraction
  translucencyLightingMode=TLM_VolumetricDirectional
  twoSided=true
  usedShadingModels=Default Lit
 }
 nodes [
  n1 = VectorParameter() {bUseCustomPrimitiveData=true, group=z_code, parameterName=Code_PrecomputedHitVector, primitiveDataIndex=23, defaultValue=<rgba 0 0 1 0>}
  n2 = VectorParameter() {bUseCustomPrimitiveData=true, group=z_code, parameterName=WSDeformationCenter, primitiveDataIndex=20}
  n3 = WorldPosition()
  n4 = Clamp(None=Multiply(
    A=Distance(A=@n2.RGB, B=@n3.XYZ)
    B=ScalarParameter() {bUseCustomPrimitiveData=true, defaultValue=1, group=z_code, parameterName=WSDamageRadius, primitiveDataIndex=26}
   ))
  n5 = LinearInterpolate(A=OneMinus(None=@n4), B=@n4, Alpha=ScalarParameter() {defaultValue=0.35, group=WorldPositionOffset, parameterName=InvertDamageFallof})
  n6 = StaticSwitch(
   True=ShadingPathSwitch(Default=Multiply(A=@n1.RGB, B=@n5), Mobile=Constant()) {desc=Migrated from FeatureLevelSwitch}
   False=Constant()
   Value=StaticBoolParameter() {defaultValue=true, group=z_code, parameterName=Code_UseHitEffects}
  )
  n7 = ScalarParameter() {bUseCustomPrimitiveData=true, group=WorldPositionOffset, parameterName=DistanceCullFadeAnimationOffset, primitiveDataIndex=12}
  n8 = MaterialFunctionCall!(World Position=Add(A=@n6, B=@n3.XYZ), DistanceCullFadeAnimationOffset=@n7, Use actor center instead of pivot point) {desc=Contains an interal feature level switch, materialFunction=/Game/Packages/Fortress_SharedMaterials/Base_Material_Functions/LODEffect}
  Add(A=@n6, B=@n8.Result) {constB=0}
  n9 = ScalarParameter() {defaultValue=1.1, group=Cracks, parameterName=Step1_CracksTiling}
  n10 = ScalarParameter() {defaultValue=1.8, group=Cracks, parameterName=Step2_CracksTiling}
  MaterialExpressionScalarParameter_33 = ScalarParameter() {defaultValue=0.33, group=Damage, parameterName=Step2_DamageValue}
  MaterialExpressionScalarParameter_4 = ScalarParameter() {group=z_blueprint, parameterName=BP_DamageLevel}
  n11 = Custom!(a=@MaterialExpressionScalarParameter_33, b=@MaterialExpressionScalarParameter_4) {
   code="return step(a,b);"
   outputType=CMOT_Float1
   inputs [
    {inputName=a, input {expression=&self:MaterialExpressionScalarParameter_33}}
    {inputName=b, input {expression=&self:MaterialExpressionScalarParameter_4}}
   ]
  }
  MaterialExpressionScalarParameter_36 = ScalarParameter() {defaultValue=0.66, group=Damage, parameterName=Step3_DamageValue}
  n12 = Custom(a=@MaterialExpressionScalarParameter_36, b=@MaterialExpressionScalarParameter_4) {
   code="return step(a,b);"
   outputType=CMOT_Float1
   inputs [
    {inputName=a, input {expression=&self:MaterialExpressionScalarParameter_36}}
    {inputName=b, input {expression=&self:MaterialExpressionScalarParameter_4}}
   ]
  }
  n13 = Multiply(
   A=TextureCoordinate()
   B=LinearInterpolate(
    A=LinearInterpolate(A=@n9, B=@n10, Alpha=@n11)
    B=ScalarParameter() {defaultValue=2.3, group=Cracks, parameterName=Step3_CracksTiling}
    Alpha=@n12
   )
  )
  BumpOffset(Coordinate=@n13, Height=ScalarParameter() {group=Parallax, parameterName=Height3}) {heightRatio=-0.4, referencePlane=0}
  n14 = CollectionParameter() {
   collection=&FortniteMaterialParameters
   parameterName=SunlightDirectionalVector
  }
  n15 = ComponentMask(None=@n14) {b=true, g=true, r=true}
  DotProduct(A=@n15, B=VertexNormalWS())
  n16 = PixelNormalWS()
  Multiply(A=@n16, B=TwoSidedSign())
  n17 = DotProduct(A=@n15, B=MaterialFunctionCall() {materialFunction=&CameraDirectionVector}.Result)
  n18 = Saturate(None=@n17)
  n19 = ScalarParameter() {defaultValue=1, parameterName=SunAttenPow}
  n20 = VectorParameter() {group=SH, parameterName=SHMax, defaultValue=<rgba 1.7242 0.7135 0.6971 0.69>}
  n21 = AppendVector(A=@n20.RGB, B=@n20.RGB)
  n22 = VectorParameter() {group=SH, parameterName=SHMin, defaultValue=<rgba 1.3883 -0.7037 -0.6761 -0.7219>}
  n23 = AppendVector(A=@n22.RGB, B=@n22.RGB)
  n24 = VertexColor()
  MaterialExpressionAdd_0 = Add(A=Multiply(A=Subtract(A=@n21, B=@n23), B=AppendVector(A=@n24, B=@n24)), B=@n23)
  MaterialExpressionTransform_0 = Transform(None=MaterialFunctionCall() {materialFunction=&CameraDirectionVector}.Result) {transformSourceType=TRANSFORMSOURCE_World, transformType=TRANSFORM_Local}
  n25 = Custom!(SH=@MaterialExpressionAdd_0, Direction=@MaterialExpressionTransform_0) {
   code="//SH decode,  Direction = LightDir, SH = 2band coefficients. (encoded in Y-up)\r\nDirection = normalize(Direction);\r\nfloat4 shEvalDirection;\r\nshEvalDirection.x = 0.282095;\r\nshEvalDirection.y =0.488603 * Direction.z;\r\nshEvalDirection.z = 0.488603 * Direction.y;\r\nshEvalDirection.w =-0.488603 * Direction.x;\r\nreturn saturate(dot(SH, shEvalDirection));"
   description=SphericalHarmonicsEval
   outputType=CMOT_Float1
   inputs [
    {inputName=SH, input {expression=&self:MaterialExpressionAdd_0}}
    {inputName=Direction, input {expression=&self:MaterialExpressionTransform_0}}
   ]
  }
  n26 = Saturate(None=MaterialFunctionCall(
    Input=OneMinus(None=@n25)
    Input Low=ScalarParameter() {defaultValue=0.15, group=SH, parameterName=SHContrast_LowValue}
    Input High=ScalarParameter() {defaultValue=0.5, group=SH, parameterName=SHContrast_HighValue}
    Target Low=Constant()
    Target High=Constant() {r=1}
   ) {materialFunction=&RemapValueRange}.Result)
  n27 = ScalarParameter() {defaultValue=2.2, group=SH, parameterName=SHPow2}
  n28 = BumpOffset(Coordinate=@n13, Height=ScalarParameter() {defaultValue=0.4, group=Parallax, parameterName=Height1}) {heightRatio=-0.4, referencePlane=0}
  n29 = TextureObjectParameter() {group=Textures, parameterName=IceCracks, texture=/Game/Gadgets/Assets/FrozenCube/Textures/T_IceCracks}
  n30 = ScalarParameter() {defaultValue=8, group=Parallax, parameterName=mipBias1}
  n31 = BumpOffset(Coordinate=@n13, Height=ScalarParameter() {defaultValue=0.1, group=Parallax, parameterName=Height2}) {heightRatio=-0.4, referencePlane=0}
  n32 = ScalarParameter() {defaultValue=5, group=Parallax, parameterName=mipBias2}
  n33 = TextureSample(UVs=@n13, Tex=@n29) {texture=/Game/Athena/WinterArt/Textures/Cracks_05_M}
  n34 = Multiply(
   A=Multiply(
    A=TextureSample!(UVs=@n28, Tex=@n29, Bias=@n30, Apply View MipBias) {mipValueMode=TMVM_MipBias, texture=&Cracks_05_M}.G
    B=TextureSample(UVs=@n31, Tex=@n29, Bias=@n32) {mipValueMode=TMVM_MipBias, texture=&Cracks_05_M}.G
   )
   B=@n33.G
  )
  n35 = CollectionParameter() {
   collection=&FortniteMaterialParameters
   parameterName=SunlightDirectionalVector
  }
  n36 = ComponentMask(None=@n35) {b=true, g=true, r=true}
  n37 = MaterialFunctionCall() {materialFunction=&CameraDirectionVector}
  n38 = Multiply(
   A=Power(Base=OneMinus(None=@n34), Exp=ScalarParameter() {defaultValue=1.1, group=Cracks, parameterName=CracksPow})
   B=Saturate(None=DotProduct(A=@n36, B=@n37.Result))
  )
  n39 = Max(
   A=ScalarParameter() {group=z_blueprint, parameterName=BP_DamageLevel}
   B=ScalarParameter() {defaultValue=0.5, group=Cracks, parameterName=MinDamage}
  )
  n40 = Saturate(None=Add(A=Multiply(A=@n38, B=@n39)) {constB=0})
  n41 = LinearInterpolate(
   A=LinearInterpolate(
    A=VectorParameter() {group=Color, parameterName=DeepColor, defaultValue=<rgba 0.009 0.01 0.0312 0>}.RGB
    B=VectorParameter() {group=Color, parameterName=DeepColorB, defaultValue=<rgba 0.009 0.01 0.0312 0>}.RGB
    Alpha=Power(Base=@n26, Exp=@n27)
   )
   B=VectorParameter() {group=Color, parameterName=SurfaceColor, defaultValue=<rgba 0.2083 0.4756 1 0>}.RGB
   Alpha=@n40
  )
  Multiply(A=Max(A=Power(Base=@n18, Exp=@n19), B=ScalarParameter() {parameterName=SunAtten}), B=@n41)
  Multiply(A=@n17) {constB=-1}
  n42 = Power(Base=@n26, Exp=ScalarParameter() {defaultValue=2.2, group=SH, parameterName=SHPow})
  OneMinus(None=@n42)
  n43 = OneMinus(None=@n33.G)
  n44 = ScalarParameter() {group=Snow, parameterName=CracksSnowMaskContrastStart}
  n45 = MaterialFunctionCall() {materialFunction=&ObjectLocalBounds}
  n46 = SphereMask(
   A=Multiply(A=Add(A=@n45.Local Bounds Minimum, B=@n45.Local Bounds Minimum)) {constB=0.5}
   B=MaterialFunctionCall() {materialFunction=&LocalPosition}."Local Position (Excluding Offsets)"
   Radius=ScalarParameter() {defaultValue=141, group=Snow, parameterName=SnowMaskRadius}
   Hardness=Constant() {r=0.5}
  )
  n47 = MaterialFunctionCall() {materialFunction=&ObjectLocalBounds}
  n48 = Divide(
   A=Subtract(
    A=MaterialFunctionCall() {materialFunction=&LocalPosition}."Local Position (Excluding Offsets)"
    B=@n47.Local Bounds Minimum
   )
   B=@n47.Local Bounds Size
  )
  n49 = Multiply(
   A=OneMinus(None=Saturate(None=@n46))
   B=Power(Base=ComponentMask(None=@n48) {b=true}, Exp=ScalarParameter() {group=Snow, parameterName=SnowMaskVerticalProgression})
  )
  n50 = MaterialFunctionCall!(
   WorldPosition=MaterialFunctionCall() {materialFunction=&LocalPosition}."Local Position (Excluding Offsets)"
   World Space Normal=VertexInterpolator(VS=PreSkinnedNormal()).PS
   UVSize_Front
   UVSize_Side
   UVSize_Top
   ProjectionTransitionContrast=ScalarParameter() {group=Snow, parameterName=SnowTriplanarContrast}
  ) {materialFunction=/Game/Packages/Fortress_SharedMaterials/Base_Material_Functions/WorldAlignedTexture_UVsMasks}
  n51 = ScalarParameter() {defaultValue=0.5, group=Snow, parameterName=SnowMaskTiling}
  n52 = Multiply(A=@n50.UV_Front, B=@n51)
  n53 = TextureObject() {samplerType=SAMPLERTYPE_Grayscale, texture=/Game/Environments/Apollo/Sets/Terrain/Textures/T_Apollo_SnowMask}
  n54 = Multiply(A=@n50.UV_Side, B=@n51)
  n55 = Add(
   A=Multiply(A=TextureSample!(UVs=@n52, Tex=@n53, Apply View MipBias).G, B=@n50.Mask_Front)
   B=Multiply(A=TextureSample(UVs=@n54, Tex=@n53).G, B=@n50.Mask_Side)
  )
  n56 = Multiply(A=TextureSample(UVs=Multiply(A=@n50.UV_Top, B=@n51), Tex=@n53).G, B=@n50.Mask_Top)
  n57 = Add(A=Subtract(A=@n49) {constB=0.5}, B=Multiply(A=@n49, B=Add(A=@n55, B=@n56)))
  n58 = Constant() {r=0.5}
  n59 = Multiply(A=ScalarParameter() {defaultValue=0.04, group=Snow, parameterName=SnowMaskSharpness, sliderMax=0.5}) {constB=0.5}
  n60 = Subtract(A=@n58, B=@n59)
  n61 = Add(A=@n58, B=@n59)
  n62 = Saturate(None=Add(
    A=Power(Base=@n43, Exp=@n44)
    B=MaterialFunctionCall(Alpha=@n57, Min=@n60, Max=@n61) {materialFunction=&SmoothStep}.Result
   ))
  n63 = Add(A=@n41, B=@n62) {constB=0}
  n64 = TextureSampleParameter2D() {
   constCoordinate=1
   group=Textures
   parameterName=SpecularMasks
   samplerSource=SSM_Wrap_WorldGroupSettings
   samplerType=SAMPLERTYPE_Masks
   sortPriority=1
   texture=/Game/Gadgets/Assets/FrozenCube/Textures/T_Frozen_Cube_S
  }
  n65 = ScalarParameter() {defaultValue=1, group=Shading, parameterName=SpecularMultiplier}
  n66 = Multiply(A=@n64.B, B=ScalarParameter() {defaultValue=1, group=Shading, parameterName=RoughnessMultiplier})
  n67 = ScalarParameter() {defaultValue=1, group=Shading, parameterName=SnowRoughness}
  n68 = ScalarParameter() {defaultValue=0.3, group=CozyLodgeMods, parameterName=AddBaseColorToEmissiveAmount, sliderMax=1}
  n69 = Saturate(None=TwoSidedSign())
  n70 = CameraVectorWS()
  n71 = Subtract(A=OneMinus(None=DotProduct(A=@n70, B=@n16)), B=ScalarParameter() {group=Fresnel, parameterName=EmissiveFalloff}) {constB=0.35}
  n72 = ScalarParameter() {defaultValue=5, group=Fresnel, parameterName=EmissvePower}
  n73 = Multiply(
   A=VectorParameter() {group=Color, parameterName=FresnelColor, defaultValue=<rgba 0.2083 0.4756 1 0>}.RGB
   B=Saturate(None=Power(Base=@n71, Exp=@n72))
  )
  n74 = Add(A=Multiply(A=Multiply(A=@n68, B=@n63), B=OneMinus(None=@n69)), B=@n73)
  n75 = VertexInterpolator(VS=@n42)
  n76 = ScalarParameter() {defaultValue=1, group=Parallax, parameterName=ParallaxMaskTranslucencyContrast}
  n77 = Add(A=Add(A=ComponentMask(None=@n75.PS) {r=true}, B=Power(Base=@n40, Exp=@n76)), B=@n62)
  n78 = TextureSample!(
   UVs
   Tex
   Bias=LinearInterpolate(A=Constant(), B=ScalarParameter() {group=Snow, parameterName=BaseNormalMipBiasUnderSnow}, Alpha=@n62)
   Apply View MipBias
  ) {mipValueMode=TMVM_MipBias, samplerType=SAMPLERTYPE_Normal, texture=/Game/Gadgets/Assets/FrozenCube/Textures/T_Frozen_Cube_N}
  n79 = AppendVector(A=@n78.R, B=@n78.R)
  n80 = ScalarParameter() {defaultValue=1, group=Shading, parameterName=BaseNormalMultiplier}
  n81 = Saturate(None=AppendVector(A=Multiply(A=@n79, B=@n80), B=@n78.B))
  n82 = TextureSampleParameter2D(UVs=@n13) {
   group=Textures
   parameterName=IceCracksNormals
   samplerType=SAMPLERTYPE_Normal
   texture=/Game/Gadgets/Assets/FrozenCube/Textures/Cracks_051_N
  }
  n83 = Multiply(
   A=AppendVector(A=@n82.R, B=@n82.R)
   B=LinearInterpolate(
    A=ScalarParameter() {defaultValue=1, group=Cracks, parameterName=CracksNormalStrengthStart}
    B=ScalarParameter() {defaultValue=1, group=Cracks, parameterName=CracksNormalStrengthEnd}
    Alpha=ScalarParameter() {group=z_blueprint, parameterName=BP_DamageLevel}
   )
  )
  n84 = ScalarParameter() {defaultValue=10, group=Snow, parameterName=SnowNormalsTiling}
  n85 = TextureObject() {samplerType=SAMPLERTYPE_Normal, texture=/Game/Abilities/Toys/Snowball/Textures/T_Snowball_Normals}
  n86 = TextureSample!(UVs=Multiply(A=@n50.UV_Front, B=@n84), Tex=@n85, Apply View MipBias)
  n87 = ScalarParameter() {defaultValue=1, group=Snow, parameterName=SnowNormalStrength}
  n88 = Multiply(
   A=Multiply(
    A=MaterialFunctionCall(Y=@n86.R, Z=@n86.R) {materialFunction=&MakeFloat3}.Result
    B=@n50.Mask_Front
   )
   B=@n87
  )
  n89 = TextureSample(UVs=Multiply(A=@n50.UV_Side, B=@n84), Tex=@n85)
  n90 = Multiply(
   A=Multiply(
    A=MaterialFunctionCall(X=@n89.R, Z=@n89.R) {materialFunction=&MakeFloat3}.Result
    B=@n50.Mask_Side
   )
   B=@n87
  )
  n91 = TextureSample(UVs=Multiply(A=@n50.UV_Top, B=@n84), Tex=@n85)
  n92 = Multiply(
   A=MaterialFunctionCall(X=@n91.R, Y=@n91.R) {materialFunction=&MakeFloat3}.Result
   B=@n50.Mask_Top
  )
  n93 = Add(A=Add(A=Add(A=@n88, B=@n90), B=Multiply(A=@n92, B=@n87)), B=VertexInterpolator(VS=PreSkinnedNormal()).PS)
  n94 = LinearInterpolate(
   A=Saturate(None=AppendVector(A=@n83, B=@n82.B))
   B=Transform(None=Normalize(VectorInput=@n93)) {transformSourceType=TRANSFORMSOURCE_Local, transformType=TRANSFORM_Tangent}
   Alpha=@n62
  )
  n95 = Subtract(A=VectorParameter() {parameterName=HitLocation}.RGB, B=ActorPositionWS())
  n96 = Subtract(A=Time(), B=ScalarParameter() {group=z_blueprint, parameterName=BP_WobbleTime})
  n97 = ScalarParameter() {defaultValue=0.4, group=WorldPositionOffset, parameterName=WobbleDuration}
  n98 = MaterialFunctionCall!(In=Saturate(None=Divide(A=@n96, B=@n97))) {materialFunction=/Game/Effects/Fort_Effects/Materials/_Functions/MF_Smoothstep_EaseOut}
  n99 = RotateAboutAxis(
   NormalizedRotationAxis=CrossProduct(A=Normalize(VectorInput=@n95), B=Constant3Vector() {constant=<rgba 0 0 1 0>})
   RotationAngle=Multiply(A=ScalarParameter() {group=WorldPositionOffset, parameterName=RotAngleMul}, B=Sine(None=@n98.Result) {period=1.5})
   PivotPoint=Constant3Vector()
   Position=MaterialFunctionCall() {materialFunction=&LocalPosition}.Local Position
  )
  n100 = ObjectRadius()
  n101 = Add(A=Subtract(A=WorldPosition().XYZ, B=ObjectPositionWS()), B=@n100)
  n102 = Add(
   A=Divide(A=ComponentMask(None=@n101) {b=true}, B=@n100)
   B=ScalarParameter() {defaultValue=0.8, group=WorldPositionOffset, parameterName=WPO_HeightMaskAtten, sliderMax=1}
  )
  n103 = ShadingPathSwitch(Default=Multiply(A=Multiply(A=@n99, B=@n98.Result), B=Saturate(None=@n102)), Mobile=Constant()) {desc=Migrated from FeatureLevelSwitch}
  n104 = Constant()
  n105 = StaticBoolParameter() {defaultValue=true, group=z_code, parameterName=Code_UseHitEffects}
  n106 = Constant() {r=1}
  n107 = ScalarParameter() {defaultValue=1.2, group=Shading, parameterName=RefractSeparateTranslucency}
  n108 = ScalarParameter() {group=z_blueprint, parameterName=BP_SeparateTranslucencyEnabled, primitiveDataIndex=1}
  n109 = MaterialFunctionCall(MA In=MakeMaterialAttributes!(
    BaseColor=@n63
    Metallic=@n64.G
    Specular=Multiply(A=@n64.R, B=@n65)
    Roughness=LinearInterpolate(A=@n66, B=@n67, Alpha=@n62)
    Anisotropy
    EmissiveColor=StaticSwitchParameter(True=@n74, False=@n73) {group=CozyLodgeMods, parameterName=AddBaseColorToEmissive}
    Opacity=Saturate(None=@n77)
    OpacityMask
    Normal=MaterialFunctionCall(BaseNormal=@n81, AdditionalNormal=@n94) {materialFunction=&BlendAngleCorrectedNormals}.Result
    Tangent
    WorldPositionOffset=StaticSwitch(True=@n103, False=@n104, Value=@n105)
    SubsurfaceColor
    ClearCoat
    ClearCoatRoughness
    AmbientOcclusion
    Refraction=LinearInterpolate(A=@n106, B=@n107, Alpha=@n108)
    CustomizedUVs_0
    CustomizedUVs_1
    CustomizedUVs_2
    CustomizedUVs_3
    CustomizedUVs_4
    CustomizedUVs_5
    CustomizedUVs_6
    CustomizedUVs_7
    PixelDepthOffset
    ShadingModel
    Displacement
   )) {materialFunction=&MF_CreativeMoveTool}
  n110 = GetMaterialAttributes(None=@n109.MA Out) {attributeGetTypes [#1]}
  n111 = LinearInterpolate(
   A=ScalarParameter() {defaultValue=1, group=CozyLodgeMods, parameterName=OpacityBackFaces, sliderMax=1}
   B=ScalarParameter() {defaultValue=1, group=CozyLodgeMods, parameterName=OpacityFrontFaces, sliderMax=1}
   Alpha=Saturate(None=TwoSidedSign())
  )
  StaticSwitchParameter(
   True=SetMaterialAttributes!(MaterialAttributes=@n110.MaterialAttributes, Opacity=Multiply(A=@n110.Opacity, B=@n111)) {attributeSetTypes [#1]}
   False=@n109.MA Out
  ) {group=CozyLodgeMods, parameterName=EnableSelectiveTwoSidedMask}
  Transform(None=VertexNormalWS()) {transformSourceType=TRANSFORMSOURCE_World, transformType=TRANSFORM_Local}
  VertexNormalWS()
 ]
}
